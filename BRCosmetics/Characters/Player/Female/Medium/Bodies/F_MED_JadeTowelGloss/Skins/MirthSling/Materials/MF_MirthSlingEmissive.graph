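# Material function named for the emissive of the MirthSling skin on the F_MED_JadeTowelGloss female player body.
# Its graph holds no expressions, so it defines no inputs, outputs or emissive logic.
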MaterialFunction /BRCosmetics/Characters/Player/Female/Medium/Bodies/F_MED_JadeTowelGloss/Skins/MirthSling/Materials/MF_MirthSlingEmissive
graph {expressions {}}
