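# Material function in the MirthSling skin's materials that outputs UVs on a camera-aligned virtual plane, with tiling and offset inputs.
# The camera direction vector sets the plane normal, and the doubled pre-skinned half extents, transformed to world space, set its center.
# UV scale multiplies the bounds' Z extent by the reciprocal of sign-preserved tiling, so the projection follows mesh size.
# The offset is added to the VirtualPlaneCoordinates result.
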MaterialFunction /BRCosmetics/Characters/Player/Female/Medium/Bodies/F_MED_JadeTowelGloss/Skins/MirthSling/Materials/MF_CameraAlignedVirtualPlane
graph {nodes [
  n1 = FunctionInput() {
   bUsePreviewValueAsDefault=true
   description=Preskinned local pos offset
   inputName=UV Offset
   inputType=FunctionInput_Vector2
   sortPriority=1
  }
  n2 = MaterialFunctionCall() {materialFunction=&CameraDirectionVector}
  n3 = PreSkinnedLocalBounds()
  n4 = TransformPosition(None=Add(A=@n3.Half Extents, B=@n3.Half Extents)) {transformType=TRANSFORMPOSSOURCE_World}
  n5 = FunctionInput() {bUsePreviewValueAsDefault=true, description=UV scale factor, inputName=UV Tiling, inputType=FunctionInput_Vector2, previewValue=<1 1 0 1>}
  n6 = Max(A=Abs(None=@n5)) {constB=0}
  n7 = Sign(None=@n5)
  n8 = Multiply(A=ComponentMask(None=@n3.Extents) {b=true}, B=Divide(B=Multiply(A=@n6, B=@n7)))
  FunctionOutput(None=Add(
    A=@n1
    B=MaterialFunctionCall(Plane Normal Axis=@n2.Result, Plane Center=@n4, UV Scale=@n8) {materialFunction=&VirtualPlaneCoordinates}.UVs
   )) {description=camera aligned virtual plane uvs, outputName=OutputUV}
 ]}
