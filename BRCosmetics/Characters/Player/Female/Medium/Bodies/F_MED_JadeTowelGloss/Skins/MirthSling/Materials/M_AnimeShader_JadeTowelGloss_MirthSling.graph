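# Cloth-shaded character material for the JadeTowelGloss body's MirthSling skin, built on the 2023 anime shader's Pre, Calculations and Post functions.
# Emissive adds an FX-channel inverse-Fresnel glow driven by audio amplitude to MirthSling cloth and skin colors.
# An FX2 texture supplies skin and edge masks, alongside a Fresnel mask confined by a pre-skinned local-space sphere.
# Panned, distorted noise on camera-aligned virtual-plane UVs provides an animated mask.
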Material /BRCosmetics/Characters/Player/Female/Medium/Bodies/F_MED_JadeTowelGloss/Skins/MirthSling/Materials/M_AnimeShader_JadeTowelGloss_MirthSling
graph {
 properties {
  bAutomaticallySetUsageInEditor=false
  bUseMaterialAttributes=true
  bUsedWithClothing=true
  bUsedWithMeshDeformer=true
  bUsedWithMorphTargets=true
  bUsedWithNiagaraMeshParticles=true
  bUsedWithSkeletalMesh=true
  bUsedWithStaticLighting=true
  pixelDepthOffsetMode=PDOM_Legacy
  shadingModel=MSM_Cloth
  usedShadingModels=Cloth
 }
 nodes [
  n1 = MaterialFunctionCall!(CustomUV) {materialFunction=/Game/Global/Materials/Character_Parent_2023/Anime_Parent_2023/MaterialFunctions/MFC_AnimeShader_2023_Pre}
  n2 = MaterialFunctionCall!(CustomLightVector, OverrideNormal_WS, STT=@n1.STT, Normal_Map=@n1.STT, DFL=@n1.STT, FX=@n1.STT) {materialFunction=/Game/Global/Materials/Character_Parent_2023/Anime_Parent_2023/MaterialFunctions/MFC_AnimeShader_2023_Calculations}
  n3 = ChannelMaskParameter(None=NamedRerouteDeclaration(None=@n1.FX) {name=FX_mask, nodeColor=<rgba 1 0.9375 0 1>}) {group=04_AudioReactivity, parameterName=ReactivChannel}
  n4 = ScalarParameter() {defaultValue=2, group=04_AudioReactivity, parameterName=MaskExponent, sliderMax=8}
  n5 = Fresnel() {baseReflectFraction=0, exponent=1}
  n6 = Multiply(
   A=Saturate(None=Power(Base=@n3.R, Exp=@n4) {constExponent=4})
   B=Power(Base=OneMinus(None=@n5), Exp=ScalarParameter() {defaultValue=2, group=04_AudioReactivity, parameterName=FresnelExponent, sliderMax=8}) {constExponent=4}
  )
  n7 = Saturate(None=Power(Base=MaterialFunctionCall() {materialFunction=/Game/Packages/Fortress_SharedMaterials/Base_Material_Functions/BaseCharacterShaderFunctions/Audio/MF_AudioReactivity}.Amplitude) {constExponent=3})
  n8 = Multiply(
   A=Multiply(A=Multiply(A=@n6, B=@n7), B=ScalarParameter() {defaultValue=5, group=04_AudioReactivity, parameterName=EmissiveIntensity, sliderMax=10}) {constB=5}
   B=VectorParameter() {group=04_AudioReactivity, parameterName=Emissive_Color, defaultValue=<rgba 0 1 0.7 1>}.RGB
  )
  n9 = VectorParameter() {group=MirthSling, parameterName=ClothEmissiveColor, defaultValue=<rgba 2.2 0.33 1.9212 1>}
  n10 = VectorParameter() {group=MirthSling, parameterName=SkinEmissiveColor, defaultValue=<rgba 2 0.02 0.5452 1>}
  n11 = NamedRerouteUsage()
  n12 = Constant() {r=0.5}
  n13 = Constant() {r=4}
  n14 = NamedRerouteUsage()
  n15 = LinearInterpolate(
   B=Multiply(A=LinearInterpolate(A=@n9.RGB, B=@n10.RGB, Alpha=@n11), B=LinearInterpolate(A=@n12, B=@n13, Alpha=@n14) {constA=1, constB=3})
   Alpha=NamedRerouteUsage()
  )
  n16 = NamedRerouteUsage()
  n17 = NamedRerouteUsage()
  MaterialFunctionCall!(
   Shading_Gradient=@n2.Shading_Gradient
   Threshold_Gradient=@n2.Shading_Gradient
   Specular_Gradient=@n2.Shading_Gradient
   Sheen_Mask=@n2.Shading_Gradient
   Thickness=@n2.Shading_Gradient
   DFL_Mask=@n2.Shading_Gradient
   FX=@n2.Shading_Gradient
   Metallic_Gradient=@n2.Shading_Gradient
   RimLight_Gradient=@n2.Shading_Gradient
   NormalBuffer=@n2.Shading_Gradient
   --------
   CINE_LightSources=@n2.Shading_Gradient
   CINE_LightColor=@n2.Shading_Gradient
   ---------
   Color_Lit=LinearInterpolate(A=@n1.Color_Lit, B=Constant4Vector() {constant=<rgba 0.3 0.09 0.1994 1>}.RGB, Alpha=NamedRerouteUsage())
   Color_Shaded=@n1.Color_Shaded
   SRM=@n1.Color_Shaded
   Emissive=LinearInterpolate(A=Add(A=@n8, B=@n15), B=Constant4Vector() {constant=<rgba 3 0.45 2.6198 1>}.RGB, Alpha=Multiply(A=@n16, B=@n17))
   DFL_Color=@n1.DFL_Color
   Specular_Color=@n1.Color_Shaded
   RimLight_Color=@n1.Color_Shaded
  ) {materialFunction=/Game/Global/Materials/Character_Parent_2023/Anime_Parent_2023/MaterialFunctions/MFC_AnimeShader_2023_Post}
  n18 = Multiply(
   A=LinearInterpolate(
    A=ScalarParameter() {defaultValue=2.0826, group=MirthSling, parameterName=MirthslingExponentln}
    B=ScalarParameter() {defaultValue=1.5, group=MirthSling, parameterName=MirthslingSkinExponentln}
    Alpha=NamedRerouteUsage()
   )
   B=LinearInterpolate(Alpha=NamedRerouteUsage()) {constA=1, constB=0.8}
  )
  n19 = Constant() {r=1.8}
  n20 = Constant() {r=1}
  n21 = NamedRerouteUsage()
  n22 = Saturate(None=Power(Base=Fresnel(ExponentIn=@n18), Exp=LinearInterpolate(A=@n19, B=@n20, Alpha=@n21)))
  n23 = StaticComponentMaskParameter(None=NamedRerouteUsage()) {defaultG=true, group=MirthSling, parameterName=MaskChannel}
  n24 = VertexInterpolator(VS=LocalPosition() {localOrigin=InstancePreSkinning}.XYZ)
  n25 = VectorParameter() {group=MirthSling, parameterName=CenterOffset, defaultValue=<rgba 0 0 213.3402 1>}
  n26 = ScalarParameter() {defaultValue=147.7057, group=MirthSling, parameterName=Radius}
  n27 = ScalarParameter() {group=MirthSling, parameterName=Hardness}
  NamedRerouteDeclaration(None=Multiply(A=LinearInterpolate(B=@n22, Alpha=@n23) {constB=0}, B=SphereMask(A=@n24.PS, B=@n25.RGB, Radius=@n26, Hardness=@n27) {hardnessPercent=0})) {name=EmissiveFresnel, nodeColor=<rgba 0.7344 1 0 1>}
  n28 = TextureSampleParameter2D() {
   group=MirthSling
   parameterName=EmissiveMaskInput
   samplerType=SAMPLERTYPE_Masks
   texture=/BRCosmetics/Characters/Player/Female/Medium/Bodies/F_MED_JadeTowelGloss/Skins/MirthSling/Textures/T_MirthSling_Body_FX2
  }
  NamedRerouteDeclaration(None=@n28.R) {name=SkinMask, nodeColor=<rgba 0.3359 1 0 1>}
  n29 = MaterialFunctionCall!(UV Tiling=Constant2Vector() {g=1, r=1}, UV Offset) {materialFunction=&/MF_CameraAlignedVirtualPlane}
  n30 = MaterialFunctionCall(UV Tiling=Constant() {r=2}) {materialFunction=&MF_CameraAlignedVirtualPlane}
  n31 = Panner(Coordinate=@n30.OutputUV) {speedY=0.02}
  n32 = Add(
   A=Panner(Coordinate=@n29.OutputUV) {speedY=0.08}
   B=Multiply(A=TextureSample(UVs=@n31) {samplerType=SAMPLERTYPE_Grayscale, texture=&LowResBlurredNoise}.R) {constB=0.2}
  )
  NamedRerouteDeclaration(None=Power(Base=TextureSample(UVs=@n32) {samplerType=SAMPLERTYPE_Grayscale, texture=&LowResBlurredNoise}.R) {constExponent=3}) {name=Noise, nodeColor=<rgba 0.125 0 1 1>}
  NamedRerouteDeclaration(None=@n28.G) {name=EdgeMask, nodeColor=<rgba 0 0.0391 1 1>}
 ]
}
